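# Niagara system for the CyberSecurity vehicle boost cosmetic, emitting from the car's skeletal mesh sockets.
# TrailMain002 layers six ribbon renderers per socket with waveform-pulsed highlights, while EmbersG adds glowing sprite embers and Tri1001 adds SubUV-animated triangle sprites in two spawn groups.
# All three emitters partially follow source movement, fading with age.
# User parameters drive painted-color variants, car offset and frontend-preview tuning of velocity, lifetime and size.
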NiagaraSystem /VehicleCosmetics/Boosts/Boost_CyberSecurity/FX/NS_CyberSecurity
dependencies {
 usedDataInterfaces [
  &NiagaraDataInterfaceCurve
  &NiagaraDataInterfaceArrayFloat
  &NiagaraDataInterfaceArrayFloat3
  &NiagaraDataInterfaceSpriteRendererInfo
  &NiagaraDataInterfaceColorCurve
  &NiagaraDataInterfaceSkeletalMesh
 ]
 usedDynamicInputs []
 usedModules [
  &self:SystemSpawnScript.SystemScriptSource.SystemScriptGraph.NiagaraNodeAssignment_0.SetVariables_#1
  &SystemState
 ]
}
emitters {
 EmbersG {
  emitterUpdateScript [
   &EmitterState()
   &VC_FilteredSocketCount()
   &SpawnRate(SpawnRate=30, Use Spawn Probability=false)
  ]
  particleSpawnScript [
   &VC_PaintedVariantsOverride() {enabled=false}
   &InitializeParticle(
    Color=&MakeCustomLinearColorFromBool(Bool=$User.bIsPainted, True Color=$User.PaintedColor, False Color=<rgba 0.4 0.4 1 1>)
    Lifetime Max=0.6
    Lifetime Min=0.1
    Lifetime Mode=::NewEnumerator1
    Position Offset=<-30 0 0>
    Sprite Rotation Mode=::NewEnumerator2
    Sprite Size Mode=::NewEnumerator1
    Uniform Sprite Size Max=&MakeFloatFromVector2d(Vector 2D=&MakeCustomVector2DFromBool(
      Bool=$User.bIsFrontendPreview
      True Vector2D=&Multiply_Vector2D_ByFloat(Float=5, Vector2D=<0.3 0.5>)
      False Vector2D=&Multiply_Vector2D_ByFloat(Float=5, Vector2D=<0.3 0.5>)
     ))
    Uniform Sprite Size Min=&MakeFloatFromVector2d(Vector 2D=&MakeCustomVector2DFromBool(
      Bool=$User.bIsFrontendPreview
      True Vector2D=&Multiply_Vector2D_ByFloat(Float=5, Vector2D=<0.2 0.3>)
      False Vector2D=&Multiply_Vector2D_ByFloat(Float=5, Vector2D=<0.2 0.3>)
     ))
    UsePositionOffset=true
    Write Material Random=true
   )
   &PartitionParticles(Number Of Partitions=$Emitter.PortCount)
   &SkeletalMeshLocation(
    Bone / Socket Index=$Particles.PartitionParticles.Partition
    Bone / Socket Sampling Mode=::NewEnumerator5
    MaskSpawnGroup=false
    Mesh Sampling Type=::NewEnumerator6
    Sampled Position Offset=&AddVectorToPosition(Position=<-70 0 0>, Vector=$User.DTCarOffset)
    Skeletal Mesh=$System.SkeletalMesh
   )
   &ShapeLocation(MaskSpawnGroup=false, Override Seed=false, Sphere Radius=35, Sphere Surface Distribution=1, UseOwnerScale=true)
   &AddVelocity(Velocity=&MakeCustomVectorFromBool(
     Bool=$User.bIsFrontendPreview
     True Vector=&RandomRangeVector(Minimum=<-90 0 -50>, Maximum=<-100 0 50>)
     False Vector=&RandomRangeVector(Minimum=<-90 0 -50>, Maximum=<-100 0 50>)
    ))
  ]
  particleUpdateScript [
   &ParticleState()
   &InheritSourceMovement(
    Applied Position Scale=&Multiply_VectorByFloat(
     Vector=<1 1 1>
     Float=&FloatFromCurve(FloatCurve=DataInterface() {propertyValues {Curve {keys [{interpMode=RCIM_Cubic, tangentMode=RCTM_Break, time=0, value=1}, {interpMode=RCIM_Cubic, tangentMode=RCTM_Break, time=1, value=0}]}}})
    )
    ApplyForce=false
    ApplyPosition=true
    ApplyPositionFalloff=false
   )
   &ScaleSpriteSize(Uniform Curve Sprite Scale=DataInterface() {propertyValues {Curve {keys [{interpMode=RCIM_Cubic, tangentMode=RCTM_Break, time=0, value=7}, {interpMode=RCIM_Cubic, tangentMode=RCTM_Break, time=1, value=0}]}}})
   &ScaleColor(
    Scale RGB=&VectorFromFloat(Value=&FloatFromCurve(FloatCurve=DataInterface() {propertyValues {Curve {keys [<at 0 10>,<at 1 0>]}}}, Scale Curve=$User.PaintedScalar))
    ScaleA=true
    ScaleRGB=true
   )
   &SolveForcesAndVelocity(Clamp Velocity=false, Limit Acceleration=false)
  ]
  rendererClasses [&NiagaraSpriteRendererProperties]
  renderers [{rendererClass=&NiagaraSpriteRendererProperties}]
 }
 TrailMain002 {
  emitterUpdateScript [
   &EmitterState()
   &VC_FilteredSocketCount()
   &VC_RateQualityHandler()
   &VC_SpawnRibbon_v1(Number of Ribbons=$Emitter.PortCount, SpawnRate=150)
   /SetVariables(Emitter.ContrastText1:NiagaraFloat=&Sine(Normalized Angle=$Emitter.Age, Period=0, Scale=0, Bias=0)) {isSetParameters=true}
  ]
  particleSpawnScript [
   &VC_PaintedVariantsOverride(
    Black=<rgba 0.1 0.1 0.1 1>
    BurntSienna=<rgba 0.0642 0.0198 0.0079 1>
    DefaultAlpha=10
    DefaultColor=<rgba 0.7 0.7 1 1>
    Lime=<rgba 0.525 0.9 0 1>
    Orange=<rgba 1 0.5 0 1>
    Pink=<rgba 1 0.5 0.8182 1>
    Purple=<rgba 0.4925 0 1.2 1>
    R1 Black:LinearColor=<rgba 0.5 0.5 0.5 0.01>
    R1 BurntSienna:LinearColor=<rgba 0.5 0.1846 0.1 1>
    R1 Orange:LinearColor=<rgba 2 0.6976 0 1>
    R1 Pink:LinearColor=<rgba 1 0.5 0.8182 1>
    R1 Saffron:LinearColor=<rgba 0.3 0.25 0 1>
    R2 Black:LinearColor=<rgba 0.1 0.1 0.1 1>
    R3 Black:LinearColor=<rgba 0.1 0.1 0.1 1>
    R4 Black:LinearColor=<rgba 0.2 0.2 0.2 1>
    R4 BurntSienna:LinearColor=<rgba 0.5 0.1649 0.075 1>
    R4 Cobalt:LinearColor=<rgba 0.2 0.4667 1 1>
    R4 DefaultAlpha=10
    R4 DefaultColor:LinearColor=<rgba 0.6 0.6 1 1>
    R4 Lime:LinearColor=<rgba 0.5833 1 0 1>
    R4 Orange:LinearColor=<rgba 1 0.5 0 1>
    R4 Pink:LinearColor=<rgba 1 0.5 0.8182 1>
    R4 Purple:LinearColor=<rgba 0.5812 0.22 1.1 1>
    R4 Saffron:LinearColor=<rgba 0.5 0.4167 0 1>
    R4 SkyBlue:LinearColor=<rgba 0 0.8333 1 1>
    R4 TitaniumWhite:LinearColor=<rgba 0.5 0.5 0.5 1>
    R5 Black:LinearColor=<rgba 1 1 1 1>
    R5 BurntSienna:LinearColor=<rgba 1 1 1 1>
    R5 Cobalt:LinearColor=<rgba 1 1 1 1>
    R5 Crimson:LinearColor=<rgba 1 0.0522 0.0779 1>
    R5 Orange:LinearColor=<rgba 0.6724 0.2346 0 1>
    R5 Pink:LinearColor=<rgba 0.5052 0.2526 0.4134 1>
    R5 SkyBlue:LinearColor=<rgba 0 1.3308 2 1>
    R5 TitaniumWhite:LinearColor=<rgba 1 1 1 1>
    Saffron=<rgba 0.6 0.5 0 1>
    SkyBlue=<rgba 0 0.8333 1 1>
    TitaniumWhite=<rgba 0.5 0.5 0.5 1>
    Use R1 Painted Overrides=true
    Use R2 Painted Overrides=true
    Use R3 Painted Overrides=true
    Use R4 Painted Overrides=true
    Use R5 Painted Overrides=true
   )
   &VC_InitializeRibbon_v1(
    Color=$Particles.Color
    Custom Offset=true
    Enable Skeletal Position=true
    Lifetime=&MakeCustomFloatFromBool(Bool=$User.bIsFrontendPreview, True Float=0.9, False Float=1.2)
    Local Vector Offset=&AddVectorToPosition(Position=<0 0 0>, Vector=$User.DTCarOffset)
    Offset Amount=1
    Ribbon Width=64
   )
   &AddVelocity(Velocity=&MakeCustomVectorFromBool(Bool=$User.bIsFrontendPreview, True Vector=<-1000 0 0>, False Vector=<-800 0 0>)) {module=AddVelocity001}
  ]
  particleUpdateScript [
   &ParticleState()
   &InheritSourceMovement(
    Applied Position Scale=&Multiply_VectorByFloat(
     Vector=<1 1 1>
     Float=&FloatFromCurve(FloatCurve=DataInterface() {propertyValues {Curve {keys [{interpMode=RCIM_Cubic, tangentMode=RCTM_Break, time=0.3, value=1}, {interpMode=RCIM_Cubic, tangentMode=RCTM_Break, time=1, value=0}]}}})
    )
    ApplyForce=false
    ApplyPosition=true
    ApplyPositionFalloff=false
   )
   &ScaleRibbonWidth(Ribbon Width Scale=&FloatFromCurve(FloatCurve=DataInterface() {propertyValues {Curve {keys [
         {interpMode=RCIM_Cubic, leaveTangent=10.0537, tangentMode=RCTM_Break, time=0.05, value=0}
         {interpMode=RCIM_Cubic, tangentMode=RCTM_Break, time=0.1, value=1.3}
         <at 0.5 0.8>
         <at 0.9 0>
        ]}}}))
   &SolveForcesAndVelocity(Clamp Velocity=false, Limit Acceleration=false)
   &VC_UpdateRibbon_v1(Enable Skeletal Position=true)
   /SetVariables(
    Particles.Outline:NiagaraFloat=&Multiply_Float(
     A=64
     B=&FloatFromCurve(FloatCurve=DataInterface() {propertyValues {Curve {keys [
          {interpMode=RCIM_Cubic, leaveTangent=10.0537, tangentMode=RCTM_Break, time=0.085, value=0}
          {interpMode=RCIM_Cubic, tangentMode=RCTM_Break, time=0.2, value=1.37}
          <at 0.55 0.8>
          <at 0.93 0>
         ]}}})
    )
    Particles.Secundary_R1:LinearColor=&MultiplyLinearColor(
     Linear Color=$Particles.R1 Color
     Scale Factor=&MakeVector4FromVectorAndFloat("Vector (RGB)"=&VectorFromFloat(Value=15), "Float (Alpha)"=1)
    )
   ) {isSetParameters=true}
   /SetVariables(Particles.HightLights:LinearColor=&MultiplyLinearColor(
     Linear Color=$Particles.R5 Color
     Scale Factor=&MakeVector4FromVectorAndFloat("Vector (RGB)"=<6 6 6>, "Float (Alpha)"=1)
    )) {isSetParameters=true}
   &ScaleColor(
    Scale Alpha=5
    Scale RGB=&VectorFromFloat(Value=&FloatFromCurve(FloatCurve=DataInterface() {propertyValues {Curve {keys [<at 0 1>,<at 1 0.3>]}}}))
    ScaleA=true
    ScaleRGB=true
   ) {module=ScaleColor001}
   /SetVariables(Particles.HighLightsSine1_ChanG:LinearColor=&MultiplyLinearColor(
     Linear Color=$Particles.Color
     Scale Factor=&MakeVector4FromVectorAndFloat(
      "Vector (RGB)"=&Multiply_Vector(
       A=&VectorFromFloat(Value=&Waveform("Waveform [1]"=::NewEnumerator0, "[1] Input Position"=2, "[1] Phase Scale"=2, "[1] Frequency"=1.2, "[1] Amplitude Min/Max"=<1 8>))
       B=&VectorFromFloat(Value=1)
      )
      "Float (Alpha)"=1
     )
    )) {isSetParameters=true}
   /SetVariables(Particles.HighLightsSine1_ChanR:LinearColor=&MultiplyLinearColor(
     Linear Color=$Particles.Color
     Scale Factor=&MakeVector4FromVectorAndFloat(
      "Vector (RGB)"=&Multiply_Vector(
       A=&VectorFromFloat(Value=&Waveform("Waveform [1]"=::NewEnumerator0, "[1] Input Position"=2, "[1] Phase Scale"=2, "[1] Frequency"=1, "[1] Amplitude Min/Max"=<1 6>))
       B=&VectorFromFloat(Value=1)
      )
      "Float (Alpha)"=1
     )
    )) {isSetParameters=true}
   /SetVariables(Particles.BaseFormsColor:LinearColor=&MultiplyLinearColor(
     Linear Color=$Particles.R4 Color
     Scale Factor=&MakeVector4FromVectorAndFloat(
      "Vector (RGB)"=&Multiply_VectorByFloat(
       Vector=&VectorFromFloat(Value=&FloatFromCurve(
         FloatCurve=DataInterface() {propertyValues {Curve {keys [
             <at 0 5 RCIM_Cubic>
             {interpMode=RCIM_Cubic, tangentMode=RCTM_Break, time=0.35, value=3}
             {interpMode=RCIM_Cubic, tangentMode=RCTM_Break, time=0.6, value=1.5}
            ]}}}
         Scale Curve=1.5
        ))
       Float=1
      )
      "Float (Alpha)"=1
     )
    )) {isSetParameters=true}
   &ScaleColor(ScaleA=true, ScaleRGB=true) {enabled=false, module=ScaleColor002}
   /SetVariables(Particles.Linear Color=&ColorFromCurve(ColorCurve=DataInterface() {propertyValues {
       AlphaCurve {keys [<at 0 1 RCIM_Cubic>,<at 0.5159 0 RCIM_Cubic>]}
       BlueCurve {keys [<at 0 1 RCIM_Cubic>,<at 1 1 RCIM_Cubic>]}
       GreenCurve {keys [<at 0 1 RCIM_Cubic>,<at 1 1 RCIM_Cubic>]}
       RedCurve {keys [<at 0 1 RCIM_Cubic>,<at 1 1 RCIM_Cubic>]}
      }})) {isSetParameters=true}
  ]
  rendererClasses [&NiagaraRibbonRendererProperties]
  renderers [
   {rendererClass=&NiagaraRibbonRendererProperties}
   {rendererClass=&NiagaraRibbonRendererProperties, rendererIndex=1}
   {rendererClass=&NiagaraRibbonRendererProperties, rendererIndex=2}
   {rendererClass=&NiagaraRibbonRendererProperties, rendererIndex=3}
   {rendererClass=&NiagaraRibbonRendererProperties, rendererIndex=4}
   {rendererClass=&NiagaraRibbonRendererProperties, rendererIndex=5}
  ]
 }
 Tri1001 {
  emitterUpdateScript [
   &EmitterState()
   &VC_FilteredSocketCount()
   &SpawnRate(SpawnRate=35, Use Spawn Probability=false)
   &SpawnRate(SpawnGroup=1, SpawnRate=40, Use Spawn Probability=false) {module=SpawnRate001}
  ]
  particleSpawnScript [
   &VC_PaintedVariantsOverride(
    Black=<rgba 0.0191 0.0191 0.0191 0.4>
    BurntSienna=<rgba 0.1 0.0324 0.02 0.6>
    Cobalt=<rgba 0.0296 0.0823 0.5457 0.6>
    Crimson=<rgba 0.205 0.0078 0.0078 0.25>
    DefaultColor=<rgba 0.35 0.35 0.7 0.5>
    ForestGreen=<rgba 0.0039 0.17 0.0039 0.5>
    Grey=<rgba 0.1 0.1 0.1 0.4>
    Lime=<rgba 0.0498 0.1 0 0.6>
    Orange=<rgba 0.3 0.0614 0.0062 0.5>
    Pink=<rgba 0.4 0.1174 0.3304 0.25>
    Purple=<rgba 0.1362 0.017 0.4 0.4>
    R1 Black:LinearColor=<rgba 0.3 0.3 0.3 1>
    R1 BurntSienna:LinearColor=<rgba 1 0.4481 0.3 1>
    R1 Cobalt:LinearColor=<rgba 1 1.1665 10 1>
    R1 Crimson:LinearColor=<rgba 1 0.1 0.1 1>
    R1 ForestGreen:LinearColor=<rgba 0 1 0.0505 1>
    R1 Grey:LinearColor=<rgba 0.8 0.8 0.8 1>
    R1 Lime:LinearColor=<rgba 0.4 0.8 0 1>
    R1 Orange:LinearColor=<rgba 5 1.744 0 1>
    R1 Pink:LinearColor=<rgba 2 1.6 1.8545 1>
    R1 Purple:LinearColor=<rgba 0.7052 0.5 1 1>
    R1 SkyBlue:LinearColor=<rgba 0.1073 0.3929 0.5365 1>
    Saffron=<rgba 0.3 0.2307 0 0.5>
    SkyBlue=<rgba 0 0.1996 0.3 0.3>
    TitaniumWhite=<rgba 0.3 0.3 0.3 0.3>
    Use R1 Painted Overrides=true
   )
   &InitializeParticle(
    Color=$Particles.Color
    Lifetime Max=&MakeCustomFloatFromBool(
     Bool=$User.bIsFrontendPreview
     True Float=&SelectFloatFromArray(Direct Array Index=&SpawnGroup(), Float Selection Array=DataInterface() {propertyValues {FloatData [0.7,0.7]}})
     False Float=&SelectFloatFromArray(Direct Array Index=&SpawnGroup(), Float Selection Array=DataInterface() {propertyValues {FloatData [0.8,0.8]}})
    )
    Lifetime Min:NiagaraFloat=&MakeCustomFloatFromBool(
     Bool=$User.bIsFrontendPreview
     True Float=&SelectFloatFromArray(Direct Array Index=&SpawnGroup(), Float Selection Array=DataInterface() {propertyValues {FloatData [0.8,0.8]}})
     False Float=&SelectFloatFromArray(Direct Array Index=&SpawnGroup(), Float Selection Array=DataInterface() {propertyValues {FloatData [0.4,0.4]}})
    )
    Lifetime Mode=::NewEnumerator1
    Position Offset=<-20 0 0>
    Sprite Rotation Angle=&RandomRangeFloat(Minimum=-1, Maximum=1)
    Sprite Rotation Mode=::NewEnumerator1
    Sprite Size Mode=::NewEnumerator1
    Uniform Sprite Size Max=&MakeCustomFloatFromBool(
     Bool=$User.bIsFrontendPreview
     True Float=&SelectFloatFromArray(Direct Array Index=&SpawnGroup(), Float Selection Array=DataInterface() {propertyValues {FloatData [60,63]}})
     False Float=&SelectFloatFromArray(Direct Array Index=&SpawnGroup(), Float Selection Array=DataInterface() {propertyValues {FloatData [30,63]}})
    )
    Uniform Sprite Size Min=&MakeCustomFloatFromBool(
     Bool=$User.bIsFrontendPreview
     True Float=&SelectFloatFromArray(Direct Array Index=&SpawnGroup(), Float Selection Array=DataInterface() {propertyValues {FloatData [51,51]}})
     False Float=&SelectFloatFromArray(Direct Array Index=&SpawnGroup(), Float Selection Array=DataInterface() {propertyValues {FloatData [42.5,85]}})
    )
    UsePositionOffset=true
    Write Material Random=true
   )
   &PartitionParticles(Number Of Partitions=$Emitter.PortCount)
   &SkeletalMeshLocation(
    Bone / Socket Index=$Particles.PartitionParticles.Partition
    Bone / Socket Sampling Mode=::NewEnumerator5
    MaskSpawnGroup=false
    Mesh Sampling Type=::NewEnumerator6
    Sampled Position Offset=&AddVectorToPosition(Position=<-40 0 0>, Vector=$User.DTCarOffset)
    Skeletal Mesh=$System.SkeletalMesh
   )
   &ShapeLocation(
    MaskSpawnGroup=false
    Override Seed=false
    Sphere Radius=&SelectFloatFromArray(Direct Array Index=&SpawnGroup(), Float Selection Array=DataInterface() {propertyValues {FloatData [40,35]}})
    UseOwnerScale=true
   )
   &AddVelocity(
    Velocity=&RandomRangeVector(
     Minimum=&MakeCustomVectorFromBool(
      Bool=$User.bIsFrontendPreview
      True Vector=&SelectVectorFromArray(
       Array Sampling Mode=::NewEnumerator0
       Direct Array Index=?
       Vector Selection Array=DataInterface() {propertyValues {FloatData [<-140 -20 -16>, <-25 -50 -8>]}}
       Random Seed=0
      )
      False Vector=&SelectVectorFromArray(
       Array Sampling Mode=::NewEnumerator0
       Direct Array Index=?
       Vector Selection Array=DataInterface() {propertyValues {FloatData [<0 0 0>, <0 0 0>]}}
       Random Seed=0
      )
     )
     Maximum=&MakeCustomVectorFromBool(
      Bool=$User.bIsFrontendPreview
      True Vector=&SelectVectorFromArray(
       Array Sampling Mode=::NewEnumerator0
       Direct Array Index=?
       Vector Selection Array=DataInterface() {propertyValues {FloatData [<-200 20 26>, <-80 50 20>]}}
       Random Seed=0
      )
      False Vector=&SelectVectorFromArray(
       Array Sampling Mode=::NewEnumerator0
       Direct Array Index=?
       Vector Selection Array=DataInterface() {propertyValues {FloatData [<0 0 0>, <0 0 0>]}}
       Random Seed=0
      )
     )
    )
    Velocity Speed Scale=1.5
   )
   &AccelerationForce(Acceleration=&SelectVectorFromArray(
     Direct Array Index=&SpawnGroup()
     Vector Selection Array=DataInterface() {propertyValues {FloatData [<-10 0 0>, <0 0 0>]}}
    )) {enabled=false, module=AccelerationForce001}
   &ApplyInitialForces()
  ]
  particleUpdateScript [
   &ParticleState()
   &SpriteRotationRate(Rotation Rate=&RandomRangeFloat(Minimum=0, Maximum=0.3))
   &ScaleColor(
    Scale Alpha=&FloatFromCurve(FloatCurve=DataInterface() {propertyValues {Curve {keys [<at 0.5 0.8>,<at 1 0>]}}})
    Scale RGB=&Multiply_VectorByFloat(
     Vector=<2 2 2>
     Float=&MakeCustomFloatFromBool(Bool=$User.bIsPainted, True Float=$User.ScaleColorPaintedTriang, False Float=1)
    )
    ScaleA=true
    ScaleRGB=true
   ) {enabled=false}
   &SubUVAnimation(
    "Sprite Renderer "=DataInterface() {propertyValues {SpriteRenderer=&self:Tri1001_0.NiagaraSpriteRendererProperties_0}}
    SubUV Animation Mode:ENiagara_SubUVLookupModeV2=::NewEnumerator1
    SubUV Curve=DataInterface() {propertyValues {Curve {keys [{arriveTangent=4.1785, interpMode=RCIM_Cubic, time=0, value=0}, {interpMode=RCIM_Cubic, leaveTangent=0.2086, time=1, value=1}]}}}
    UseEndFrame=false
    UseStartFrame=false
   )
   &InheritSourceMovement(
    Applied Position Scale=&Multiply_VectorByFloat(
     Vector=<1 1 1>
     Float=&FloatFromCurve(FloatCurve=DataInterface() {propertyValues {Curve {keys [{interpMode=RCIM_Cubic, tangentMode=RCTM_Break, time=0, value=1}, {interpMode=RCIM_Cubic, tangentMode=RCTM_Break, time=1, value=0}]}}})
    )
    ApplyForce=false
    ApplyPosition=true
    ApplyPositionFalloff=false
   )
   &SolveForcesAndVelocity(Clamp Velocity=false, Limit Acceleration=false)
  ]
  rendererClasses [&NiagaraSpriteRendererProperties]
  renderers [
   {rendererClass=&NiagaraSpriteRendererProperties}
   {rendererClass=&NiagaraSpriteRendererProperties, rendererIndex=1}
  ]
 }
}
summary {
 systemName=NS_CyberSecurity
 emitters [
  {emitterName=EmbersG, rendererClasses [&NiagaraSpriteRendererProperties]}
  {emitterName=Tri1001, rendererClasses [&NiagaraSpriteRendererProperties]}
  {emitterName=TrailMain002, rendererClasses [&NiagaraRibbonRendererProperties]}
 ]
 userVariables [
  User.bIsFrontendPreview {type=NiagaraBool}
  User.bIsPainted {type=NiagaraBool}
  User.ColorVariantID {type=NiagaraInt32}
  User.DTCarOffset {type=Vector3f, description="", defaultValue {struct=Vector3f, value=<0 0 0>}}
  User.InneLinesMult {type=NiagaraFloat, defaultValue=1, description=""}
  User.OutlinePainted {type=NiagaraFloat, defaultValue=0.5, description=""}
  User.PaintedColor {type=LinearColor, description="", defaultValue {struct=LinearColor, value=<rgba 1 1 1 1>}}
  User.PaintedScalar {type=NiagaraFloat, defaultValue=1, description=""}
  User.PortCount {type=NiagaraInt32, defaultValue=2}
  User.ScaleColorPaintedTriang {type=NiagaraFloat, defaultValue=1, description=""}
  User.TriOutline {type=NiagaraFloat, defaultValue=3, description=""}
 ]
}
